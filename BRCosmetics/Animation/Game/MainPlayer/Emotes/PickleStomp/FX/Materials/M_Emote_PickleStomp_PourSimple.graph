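# Masked, lit Niagara sprite material for the pour effect in the PickleStomp emote.
# Tile and pan parameters drive one panner that scrolls both the pour texture and its normal map.
# A time-of-day function takes the texture color, with only the night input tinted by a Night parameter.
# Named reroutes carry color, alpha-based opacity and normal, with scalar multipliers for base and emissive intensity.
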Material /BRCosmetics/Animation/Game/MainPlayer/Emotes/PickleStomp/FX/Materials/M_Emote_PickleStomp_PourSimple
graph {
 expressions {
  MaterialExpressionConvert_1 {
   class=MaterialExpressionConvert
   inputs [
    None {from=MaterialExpressionScalarParameter_0}
    None {from=MaterialExpressionScalarParameter_1}
   ]
   properties {
    nodeName=Make Vector2
    convertInputs [
     {
      type=Scalar
      expressionInput {
       expression=&self:MaterialExpressionScalarParameter_0
       inputName=None
       mask=0
       maskA=0
       maskB=0
       maskG=0
       maskR=0
       outputIndex=0
      }
     }
     {
      type=Scalar
      expressionInput {
       expression=&self:MaterialExpressionScalarParameter_1
       inputName=None
       mask=0
       maskA=0
       maskB=0
       maskG=0
       maskR=0
       outputIndex=0
      }
     }
    ]
    convertOutputs [{type=Vector2}]
   }
  }
  MaterialExpressionConvert_2 {
   class=MaterialExpressionConvert
   inputs [
    None {from=MaterialExpressionScalarParameter_2}
    None {from=MaterialExpressionScalarParameter_3}
   ]
   properties {
    nodeName=Make Vector2
    convertInputs [
     {
      type=Scalar
      expressionInput {
       expression=&self:MaterialExpressionScalarParameter_2
       inputName=None
       mask=0
       maskA=0
       maskB=0
       maskG=0
       maskR=0
       outputIndex=0
      }
     }
     {
      type=Scalar
      expressionInput {
       expression=&self:MaterialExpressionScalarParameter_3
       inputName=None
       mask=0
       maskA=0
       maskB=0
       maskG=0
       maskR=0
       outputIndex=0
      }
     }
    ]
    convertOutputs [{type=Vector2}]
   }
  }
  MaterialExpressionMaterialFunctionCall_0 {
   class=MaterialExpressionMaterialFunctionCall
   inputs [
    Color_Morning {from=MaterialExpressionTextureSampleParameter2D_1, output=RGB}
    Color_Day {from=MaterialExpressionTextureSampleParameter2D_1, output=RGB}
    Color_Evening {from=MaterialExpressionTextureSampleParameter2D_1, output=RGB}
    Color_Night {from=MaterialExpressionMultiply_3}
   ]
   properties {materialFunction=&Color_Time_of_Day}
  }
  MaterialExpressionMultiply_0 {
   class=MaterialExpressionMultiply
   inputs [
    A {from=MaterialExpressionTextureCoordinate_0}
    B {from=MaterialExpressionConvert_1}
   ]
  }
  MaterialExpressionMultiply_1 {
   class=MaterialExpressionMultiply
   inputs [
    A {from=MaterialExpressionNamedRerouteUsage_0}
    B {from=MaterialExpressionScalarParameter_4}
   ]
  }
  MaterialExpressionMultiply_2 {
   class=MaterialExpressionMultiply
   inputs [
    A {from=MaterialExpressionNamedRerouteUsage_1}
    B {from=MaterialExpressionScalarParameter_5}
   ]
  }
  MaterialExpressionMultiply_3 {
   class=MaterialExpressionMultiply
   inputs [
    A {from=MaterialExpressionTextureSampleParameter2D_1, output=RGB}
    B {from=MaterialExpressionVectorParameter_0, output=RGB}
   ]
  }
  MaterialExpressionNamedRerouteDeclaration_0 {
   class=MaterialExpressionNamedRerouteDeclaration
   inputs [None {from=MaterialExpressionMaterialFunctionCall_0, output=Result}]
   properties {name=Color, nodeColor=<rgba 0 0.0625 1 1>}
  }
  MaterialExpressionNamedRerouteDeclaration_1 {
   class=MaterialExpressionNamedRerouteDeclaration
   inputs [None {from=MaterialExpressionReroute_1}]
   properties {name=Opac, nodeColor=<rgba 0.3828 1 0 1>}
  }
  MaterialExpressionNamedRerouteDeclaration_2 {
   class=MaterialExpressionNamedRerouteDeclaration
   inputs [None {from=MaterialExpressionTextureSampleParameter2D_0, output=RGB}]
   properties {name=Norm, nodeColor=<rgba 1 0 0.7031 1>}
  }
  MaterialExpressionNamedRerouteUsage_0 {class=MaterialExpressionNamedRerouteUsage}
  MaterialExpressionNamedRerouteUsage_1 {class=MaterialExpressionNamedRerouteUsage}
  MaterialExpressionNamedRerouteUsage_2 {class=MaterialExpressionNamedRerouteUsage}
  MaterialExpressionNamedRerouteUsage_3 {class=MaterialExpressionNamedRerouteUsage}
  MaterialExpressionPanner_0 {
   class=MaterialExpressionPanner
   inputs [
    Coordinate {from=MaterialExpressionMultiply_0}
    Time {from=""}
    Speed {from=MaterialExpressionConvert_2}
   ]
   properties {speedY=-1}
  }
  MaterialExpressionReroute_1 {class=MaterialExpressionReroute, inputs [None {from=MaterialExpressionTextureSampleParameter2D_1, output=A}]}
  MaterialExpressionScalarParameter_0 {class=MaterialExpressionScalarParameter, properties {defaultValue=1, parameterName=UTile}}
  MaterialExpressionScalarParameter_1 {class=MaterialExpressionScalarParameter, properties {defaultValue=1, parameterName=VTile}}
  MaterialExpressionScalarParameter_2 {class=MaterialExpressionScalarParameter, properties {parameterName=UPan}}
  MaterialExpressionScalarParameter_3 {class=MaterialExpressionScalarParameter, properties {parameterName=VPan}}
  MaterialExpressionScalarParameter_4 {class=MaterialExpressionScalarParameter, properties {defaultValue=1, parameterName=BaseIntens}}
  MaterialExpressionScalarParameter_5 {class=MaterialExpressionScalarParameter, properties {parameterName=EmisIntens}}
  MaterialExpressionTextureCoordinate_0 {class=MaterialExpressionTextureCoordinate}
  MaterialExpressionTextureSampleParameter2D_0 {
   class=MaterialExpressionTextureSampleParameter2D
   inputs [
    UVs {from=MaterialExpressionPanner_0, output=""}
    Apply View MipBias {from="", output=""}
   ]
   properties {
    parameterName=TexNormal
    samplerType=SAMPLERTYPE_Normal
    texture=/BRCosmetics/Animation/Game/MainPlayer/Emotes/PickleStomp/FX/Texture/T_Emote_PickleStomp_PourNorm
   }
  }
  MaterialExpressionTextureSampleParameter2D_1 {
   class=MaterialExpressionTextureSampleParameter2D
   inputs [
    UVs {from=MaterialExpressionPanner_0, output=""}
    Apply View MipBias {from="", output=""}
   ]
   properties {parameterName=Tex, texture=/BRCosmetics/Animation/Game/MainPlayer/Emotes/PickleStomp/FX/Texture/T_Emote_PickleStomp_Pour}
  }
  MaterialExpressionVectorParameter_0 {class=MaterialExpressionVectorParameter, properties {parameterName=Night, defaultValue=<rgba 1 1 1 1>}}
 }
 properties {bAutomaticallySetUsageInEditor=false, bUsedWithNiagaraSprites=true, blendMode=BLEND_Masked, usedShadingModels=Default Lit}
}
