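# Niagara system for the pouring liquid effect in the PickleStomp emote, built from two sprite emitters.
# Both spawn in bursts at offsets below a location sampled through a shared system Skeletal Mesh Reader.
# Stream makes short-lived elongated sprites whose width swells then narrows.
# Splash launches droplets tinted by User.WaterColor in an upward cone with randomized size, lifetime and speed, then applies drag, gravity and an alpha fade.
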NiagaraSystem /BRCosmetics/Animation/Game/MainPlayer/Emotes/PickleStomp/FX/NS_Emote_PickleStomp_Pour
dependencies {
 usedDataInterfaces [
  &NiagaraDataInterfaceCurve
  &NiagaraDataInterfaceVector2DCurve
  &NiagaraDataInterfaceSkeletalMesh
 ]
 usedDynamicInputs [&Waveform, &Multiply_Vector]
 usedModules [
  &self:SystemSpawnScript.SystemScriptSource.SystemScriptGraph.NiagaraNodeAssignment_0.SetVariables_#1
  &SystemState
  &self:SystemSpawnScript.SystemScriptSource.SystemScriptGraph.NiagaraNodeAssignment_2.SetVariables_#2
  &self:SystemSpawnScript.SystemScriptSource.SystemScriptGraph.NiagaraNodeAssignment_3.SetVariables_#3
 ]
}
emitters {
 Splash {
  emitterUpdateScript [
   &EmitterState()
   &SpawnBurst_Instantaneous(Spawn Count=15, Use Spawn Probability=false, UseLoopCountLimit=false)
   /SetVariables(Emitter.FrameRand:NiagaraFloat=&RandomRangeFloat(Minimum=0.5, Maximum=1)) {isSetParameters=true}
   &SpawnRate(SpawnRate=&RandomRangeFloat(Minimum=100, Maximum=600), Use Spawn Probability=false) {enabled=false}
  ]
  particleSpawnScript [
   /Niagara/Modules/Spawn/Initialization/InitializeParticle(
    Lifetime=5
    Module.Write Color=true
    Module.Write Lifetime=true
    Module.Write Mass=true
    Module.Write Material Random=true
    Module.Write Position=false
    Module.Write Scale=false
    Module.Write SpriteRotation=false
    Module.Write SpriteSize=true
    Position=&SimulationPosition()
    Sprite Size=<5 5>
   ) {enabled=false}
   /Niagara/Modules/Spawn/Initialization/V2/InitializeParticle(
    Color=$User.WaterColor
    Lifetime=&FloatFromCurve(
     FloatCurve=DataInterface() {propertyValues {Curve {keys [
         {arriveTangent=-0.1485, interpMode=RCIM_Cubic, leaveTangent=-0.1485, tangentMode=RCTM_User, time=0, value=1}
         {arriveTangent=-1.6896, interpMode=RCIM_Cubic, leaveTangent=-1.6896, tangentMode=RCTM_User, time=1, value=0.2}
        ]}}}
     CurveIndex=&RandomRangeFloat(Minimum=0, Maximum=1)
     Scale Curve=0.5
    )
    Sprite Size=&Vector2DFromCurve(
     Vector2Curve=DataInterface() {propertyValues {
       XCurve {keys [<at -0.0011 5 RCIM_Cubic>,<at 0.9988 3 RCIM_Cubic>]}
       YCurve {keys [
         {arriveTangent=-30.8979, interpMode=RCIM_Cubic, leaveTangent=-30.8979, tangentMode=RCTM_User, time=0, value=15.3962}
         {arriveTangent=-6.8193, interpMode=RCIM_Cubic, leaveTangent=-6.8193, tangentMode=RCTM_User, time=0.2382, value=8}
         <at 0.9988 4 RCIM_Cubic>
        ]}
      }}
     CurveIndex=&RandomRangeFloat(Minimum=0, Maximum=1)
     ScaleCurve=<0.4 0.7>
    )
    Sprite Size Mode=::NewEnumerator4
    UsePositionOffset=false
    Write Material Random=true
   ) {module=InitializeParticle001}
   &SkeletalMeshLocation(
    MaskSpawnGroup=false
    Mesh Sampling Type=::NewEnumerator5
    Orientation Sampling=::NewEnumerator1
    Sampled Position Offset=<0 0 -15>
    Skeletal Mesh=$System.Skeletal Mesh Reader
    Socket Sampling Mode=::NewEnumerator5
   )
   &AddVelocity(
    Cone Angle=80
    Cone Axis=<0 0 1>
    Module.UseConeFalloff=false
    Module.UseVelDistribution=false
    Override Seed=false
    Velocity Mode=::NewEnumerator2
    Velocity Speed=&Multiply_Float(
     A=$Emitter.FrameRand
     B=&RandomRangeFloat(Minimum=100, Maximum=&RandomRangeFloat(Minimum=150, Maximum=300))
    )
   ) {module=AddVelocity001}
  ]
  particleUpdateScript [
   &ParticleState()
   &ScaleColor(
    Scale Alpha=&FloatFromCurve(
     FloatCurve=DataInterface() {propertyValues {Curve {keys [
         <at 0 0 RCIM_Cubic>
         {arriveTangent=0.002, interpMode=RCIM_Cubic, leaveTangent=0.002, time=0.1183, value=1.0113}
         <at 1.0011 0.002 RCIM_Cubic>
        ]}}}
     Scale Curve=0.6
    )
    Scale RGB=<0.7 1 0.3>
    ScaleA=true
    ScaleRGB=true
   )
   &Drag(Drag=&RandomRangeFloat(Minimum=1, Maximum=1.6), Use Linear Drag=true, Use Rotational Drag=false)
   &GravityForce()
   &SolveForcesAndVelocity(Clamp Velocity=false, Limit Acceleration=false)
  ]
  rendererClasses [&NiagaraSpriteRendererProperties]
  renderers [{rendererClass=&NiagaraSpriteRendererProperties}]
 }
 Stream {
  emitterUpdateScript [
   &EmitterState()
   &SpawnBurst_Instantaneous(Use Spawn Probability=false, UseLoopCountLimit=false)
  ]
  particleSpawnScript [
   &InitializeParticle(
    Color Mode=::NewEnumerator0
    Lifetime=0.15
    Mesh Scale=<3.5 15 1>
    Mesh Scale Mode=::NewEnumerator4
    Sprite Rotation Angle=90
    Sprite Rotation Mode=::NewEnumerator0
    Sprite Size=<3.5 10>
    Sprite Size Mode=::NewEnumerator4
    UsePositionOffset=false
    Write Material Random=true
   )
   &SkeletalMeshLocation(
    MaskSpawnGroup=false
    Mesh Sampling Type=::NewEnumerator5
    Orientation Sampling=::NewEnumerator1
    Sampled Position Offset=<0 0 -10>
    Skeletal Mesh=$System.Skeletal Mesh Reader
    Socket Sampling Mode=::NewEnumerator5
   )
   &ApplyInitialForces()
  ]
  particleUpdateScript [
   &ParticleState()
   &ScaleSpriteSize(
    Non-Uniform Curve Sprite Scale=DataInterface() {propertyValues {XCurve {keys [<at 0.0157 0.5>,<at 0.2066 1.25>,<at 0.5033 1.25>,<at 1 0.2>]}, YCurve {keys [<at 0 1>]}}}
    Scale Sprite Size Mode=::NewEnumerator3
   )
  ]
  rendererClasses [&NiagaraSpriteRendererProperties]
  renderers [{rendererClass=&NiagaraSpriteRendererProperties}]
 }
}
summary {
 systemName=NS_Emote_PickleStomp_Pour
 emitters [
  {emitterName=Stream, rendererClasses [&NiagaraSpriteRendererProperties]}
  {emitterName=Splash, rendererClasses [&NiagaraSpriteRendererProperties]}
 ]
 userVariables [
  User.Audio Player {
   type=NiagaraDataInterfaceAudioPlayer
   description=""
   defaultValue {
    struct=NiagaraExt_VariableValue_DataInterface
    value {
     dataInterface=&self:NiagaraDataInterfaceAudioPlayer_0
     dataInterfaceClass=&NiagaraDataInterfaceAudioPlayer
    }
   }
  }
  User.PlayWaterSplash {type=NiagaraBool, defaultValue=true}
  User.SplashLocation {type=NiagaraPosition, description="", defaultValue {}}
  User.WaterColor {type=LinearColor, description="", defaultValue {struct=LinearColor, value=<rgba 0.6 0.4745 0.3299 1>}}
 ]
}
