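# Niagara system for the PickleStomp emote, named as the prop-spawn effect, that plays a brief mesh splash.
# Splash_Mesh bursts a few short-lived, yellow-tinted meshes with random orientation, placed using a system-sampled skeletal mesh position and System.SplashPosition.
# Over each lifetime, curves scale the mesh outward, fade alpha and drive four dynamic material parameters.
# The Droplets, Splashes, SmokeLWE and SparklesLWE emitters show no renderers or modules here.
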NiagaraSystem /BRCosmetics/Animation/Game/MainPlayer/Emotes/PickleStomp/FX/NS_Emote_PickleStomp_PropSpawn
dependencies {
 usedDataInterfaces [&NiagaraDataInterfaceCurve, &NiagaraDataInterfaceVectorCurve]
 usedDynamicInputs [&FNTimeOfDayColor]
 usedModules [
  &self:SystemSpawnScript.SystemScriptSource.SystemScriptGraph.NiagaraNodeAssignment_0.SetVariables_#1
  &SystemState
 ]
}
emitters {
 Droplets {rendererClasses [], renderers []}
 SmokeLWE {rendererClasses [], renderers []}
 SparklesLWE {rendererClasses [], renderers []}
 Splash_Mesh {
  emitterUpdateScript [
   &EmitterState()
   &SpawnBurst_Instantaneous(Spawn Count=4, Use Spawn Probability=false, UseLoopCountLimit=false)
  ]
  particleSpawnScript [
   &InitializeParticle(
    Color=<rgba 0.5811 0.4236 0.0515 1>
    Lifetime=0.6
    Mesh Scale=<0.08 0.08 0.12>
    Mesh Scale Mode=::NewEnumerator4
    Position=$System.SampleSkeletalMesh.SampledPosition
    Position Mode=::NewEnumerator1
    Position Offset=<0 0 -15>
    UsePositionOffset=true
    Write Material Random=true
   ) {module=InitializeParticle001}
   &InitialMeshOrientation(
    Mesh Orientation Mode=::NewEnumerator1
    Rotation=&RandomRangeVector(Minimum=<0 0 0>, Maximum=<0 0 1>)
    Use Rotation Vector=true
   )
   &ShapeLocation(
    MaskSpawnGroup=false
    Offset=&Add_Vector(A=<0 0 -20>, B=Linked() {linkedVariable {name=System.SplashPosition, type=NiagaraPosition}})
    Offset Mode=::NewEnumerator0
    Override Seed=false
    Sphere Radius=0
    UseOwnerScale=true
   )
  ]
  particleUpdateScript [
   &ParticleState()
   &DynamicMaterialParameters(
    Index 0 Param 1=&FloatFromCurve(
     FloatCurve=DataInterface() {propertyValues {Curve {keys [{tangentMode=RCTM_User, time=0, value=1}, {tangentMode=RCTM_User, time=1, value=0}]}}}
     Scale Curve=-2
    )
    Index 0 Param 2=&FloatFromCurve(FloatCurve=DataInterface() {propertyValues {Curve {keys [<at -0.0046 4>,<at 1.0011 0>]}}})
    Index 0 Param 3=&FloatFromCurve(FloatCurve=DataInterface() {propertyValues {Curve {keys [
         {arriveTangent=4.1785, interpMode=RCIM_Cubic, leaveTangent=4.1785, tangentMode=RCTM_User, time=0, value=0}
         {arriveTangent=0.9995, interpMode=RCIM_Cubic, leaveTangent=0.9995, tangentMode=RCTM_User, time=0.2223, value=0.6825}
         {arriveTangent=-0.0405, interpMode=RCIM_Cubic, leaveTangent=-0.0405, tangentMode=RCTM_User, time=1, value=1}
        ]}}})
    Index 0 Param 4=&FloatFromCurve(FloatCurve=DataInterface() {propertyValues {Curve {keys [<at 0.2 0 RCIM_Cubic>, {interpMode=RCIM_Cubic, leaveTangent=3.4641, time=1, value=1}]}}})
    Module.Param0WriteEnabled=true
    Module.Param1WriteEnabled=true
    Module.Param2WriteEnabled=true
    Module.Param3WriteEnabled=true
   ) {module=DynamicMaterialParameters001}
   &ScaleColor(
    Scale Alpha=&FloatFromCurve(FloatCurve=DataInterface() {propertyValues {Curve {keys [{interpMode=RCIM_Cubic, tangentMode=RCTM_User, time=0, value=0.7}, {interpMode=RCIM_Cubic, tangentMode=RCTM_User, time=1, value=0}]}}})
    ScaleA=true
    ScaleRGB=false
   )
   &ScaleMeshSize(Scale Factor=&VectorFromCurve(VectorCurve=DataInterface() {propertyValues {
       XCurve {keys [{arriveTangent=2.6672, interpMode=RCIM_Cubic, leaveTangent=2.6672, tangentMode=RCTM_User, time=0, value=1.5}, <at 0.9989 2.5 RCIM_Cubic>]}
       YCurve {keys [{arriveTangent=2.6672, interpMode=RCIM_Cubic, leaveTangent=2.6672, tangentMode=RCTM_User, time=0, value=1.5}, <at 0.9989 2.5 RCIM_Cubic>]}
       ZCurve {keys [
         {arriveTangent=1.1096, interpMode=RCIM_Cubic, leaveTangent=1.1096, tangentMode=RCTM_User, time=0, value=1}
         <at 0.5 1.25 RCIM_Cubic>
         {arriveTangent=-1.0344, interpMode=RCIM_Cubic, leaveTangent=-1.0344, tangentMode=RCTM_User, time=0.9989, value=1}
        ]}
      }})) {module=MeshSizeScale}
  ]
  rendererClasses [&NiagaraMeshRendererProperties]
  renderers [{rendererClass=&NiagaraMeshRendererProperties, rendererIndex=0}]
 }
 Splashes {rendererClasses [], renderers []}
}
summary {
 systemName=NS_Emote_PickleStomp_PropSpawn
 emitters [
  {emitterName=SparklesLWE, rendererClasses []}
  {emitterName=SmokeLWE, rendererClasses []}
  {emitterName=Droplets, rendererClasses []}
  {emitterName=Splashes, rendererClasses []}
  {emitterName=Splash_Mesh, rendererClasses [&NiagaraMeshRendererProperties]}
 ]
 userVariables [User.NormalizedNotifyProgress {type=NiagaraFloat, defaultValue=0, description=""}]
}
